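# Material function that produces animated scanlines, by its name for the ClawPad Host back bling material.
# It wraps the Atlas effect function MF_Atlas_Scanlines_01, with parameters for speed and block opacity, plus line count and darkness inputs that default to parameters.
# Time scaled by the speed parameter and a constant drives the animation.
# The output adds the secondary lines, scaled by an accent brightness parameter, to the main result.
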MaterialFunction /BRCosmetics/Accessories/FORT_Backpacks/Backpack_M_MED_ClawPad_Host/Materials/MF_Scanlines_ClawpadHost
graph {nodes [
  n1 = ScalarParameter() {defaultValue=0.004, parameterName=ScanlineSpeed}
  n2 = Time()
  n3 = MaterialFunctionCall!(
   Number of Scanlines=FunctionInput(Preview=ScalarParameter() {defaultValue=100, parameterName=Scanline amount}) {bUsePreviewValueAsDefault=true, inputName=Scanline Amount, inputType=FunctionInput_Scalar}
   Scanline Speed=@n1
   "Scanlines Thickness (Bias)"
   Scanlines Dark Value=FunctionInput(Preview=ScalarParameter() {defaultValue=0.5, parameterName=Scanline Darkness}) {bUsePreviewValueAsDefault=true, inputName=Scanline Dark Value, inputType=FunctionInput_Scalar}
   Scanlines Light Value
   Block Opacity=ScalarParameter() {defaultValue=0.2, parameterName=BlockOpacity}
   UVs
   Time=Multiply(A=Multiply(A=@n2, B=@n1)) {constB=20}
   Block Repeat=Constant() {r=3}
  ) {materialFunction=/Game/Effects/Fort_Effects/Effects/Atlas/Materials/MF_Atlas_Scanlines_01}
  n4 = ScalarParameter() {defaultValue=1, parameterName=Accent Scanlight Brightness}
  FunctionOutput(None=Add(A=@n3.Result, B=Multiply(A=@n3.Secondary Lines, B=@n4) {constB=3}))
 ]}
