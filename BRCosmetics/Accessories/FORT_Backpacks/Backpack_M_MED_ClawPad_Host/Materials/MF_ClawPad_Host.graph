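# Material function that turns input attributes into the ClawPad Host backpack's hologram look for two figures, Cody and 66.
# Masks and normals build tinted gradients, bright spots, fresnel rims, camera-faded scanlines and time-of-day-lit color.
# Visibility parameters gate each figure's opacity and collapse it toward the pivot via vertex-color-masked WPO, layered with noise and band glitches.
# Opacity supports optional translucency sorting and also drives a dithered opacity mask.
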MaterialFunction /BRCosmetics/Accessories/FORT_Backpacks/Backpack_M_MED_ClawPad_Host/Materials/MF_ClawPad_Host
graph {nodes [
  n1 = GetMaterialAttributes(None=FunctionInput() {inputName=Result, inputType=FunctionInput_MaterialAttributes}) {attributeGetTypes [
    #1
    #2
    #3
    #4
    #5
   ]}
  n2 = VectorParameter() {group=Clawpad Host, parameterName=Hologram Tint - Cody, defaultValue=<rgba 0 0.9964 2 1>}
  n3 = VectorParameter() {group=Clawpad Host, parameterName=Hologram Tint - 66, defaultValue=<rgba 0 0.9964 2 1>}
  n4 = NamedRerouteUsage()
  n5 = LinearInterpolate(
   A=ScalarParameter() {defaultValue=1.2185, group=Clawpad Host, parameterName=Hologram Tint - Min - Cody}
   B=ScalarParameter() {defaultValue=1.2185, group=Clawpad Host, parameterName=Hologram Tint - Min - 66}
   Alpha=NamedRerouteUsage()
  )
  n6 = ScalarParameter() {defaultValue=2.674, group=Clawpad Host, parameterName=Hologram Tint - Max}
  n7 = MaterialFunctionCall(Float3=Divide(A=LocalPosition().XYZ, B=PreSkinnedLocalBounds().Extents)) {materialFunction=&BreakOutFloat3Components}
  n8 = Saturate(None=OneMinus(None=LinearInterpolate(A=@n5, B=@n6, Alpha=@n7.B)))
  n9 = LinearInterpolate(
   A=LinearInterpolate(
    A=ScalarParameter() {defaultValue=1.01, group=Clawpad Host, parameterName=Hologram Tint - Min 2 - Cody}
    B=ScalarParameter() {defaultValue=1.01, group=Clawpad Host, parameterName=Hologram Tint - Min 2 - 66}
    Alpha=NamedRerouteUsage()
   )
   B=ScalarParameter() {defaultValue=2, group=Clawpad Host, parameterName=Hologram Tint - Max 2}
   Alpha=@n7.B
  )
  n10 = Saturate(None=Multiply(A=OneMinus(None=@n9), B=ScalarParameter() {defaultValue=2, group=Clawpad Host, parameterName=Bright Spot - Intensity}))
  n11 = ScalarParameter() {defaultValue=0.2439, group=Clawpad Host, parameterName=Bright Spot - Step Amount}
  n12 = DotProduct(
   A=Constant3Vector() {constant=<rgba 0 -1 0 0>}
   B=Transform(None=VertexNormalWS()) {transformSourceType=TRANSFORMSOURCE_World, transformType=TRANSFORM_Local}
  )
  n13 = SmoothStep(Min=@n11, Max=Add(A=@n11, B=ScalarParameter() {defaultValue=0.37, group=Clawpad Host, parameterName=Bright Spot - Smoothness}), Value=@n12)
  n14 = Add(
   A=LinearInterpolate(
    A=VectorParameter() {group=Clawpad Host, parameterName=Background Hologram Tint, defaultValue=<rgba 0.0454 0.1395 0.3478 1>}.RGB
    B=LinearInterpolate(A=@n2.RGB, B=@n3.RGB, Alpha=@n4)
    Alpha=@n8
   )
   B=Saturate(None=Multiply(A=@n10, B=@n13))
  )
  n15 = ScalarParameter() {defaultValue=0.5, group=Clawpad Host, parameterName=Dark Area - Darkening Amount}
  n16 = ScalarParameter() {defaultValue=9, group=Clawpad Host, parameterName=Fresnel Rim - Fresnel}
  n17 = VertexNormalWS()
  n18 = Step(X=DotProduct(
    A=Constant3Vector() {constant=<rgba 0 1 -1 0>}
    B=Transform(None=@n17) {transformSourceType=TRANSFORMSOURCE_World, transformType=TRANSFORM_Local}
   ))
  n19 = LinearInterpolate(
   A=LinearInterpolate(A=@n14, B=Multiply(A=@n14, B=@n15) {constB=0.3}, Alpha=NamedRerouteUsage())
   B=VectorParameter() {group=Clawpad Host, parameterName=Fresnel Rim - Light Color, defaultValue=<rgba 0 7.7229 20 1>}.RGB
   Alpha=Multiply(A=Fresnel(ExponentIn=@n16) {baseReflectFraction=0}, B=Saturate(None=@n18))
  )
  n20 = LinearInterpolate(
   A=ScalarParameter() {defaultValue=0.2, parameterName=Scanline - Darkness - Close}
   B=ScalarParameter() {defaultValue=0.75, parameterName=Scanline - Darkness - Far}
   Alpha=MaterialFunctionCall(
    Fade Length=ScalarParameter() {defaultValue=50, parameterName=Scanline - Darkness Fade Length}
    Fade Offset=ScalarParameter() {defaultValue=10, parameterName=Scanline - Darkness Fade Distance}
   ) {materialFunction=&CameraDepthFade}.Result
  )
  n21 = NamedRerouteDeclaration(None=MaterialFunctionCall!(
    Default=MaterialFunctionCall!(Scanline Dark Value=@n20, Scanline Amount) {materialFunction=&/MF_Scanlines_ClawpadHost}.Result
    Low/Mobile=Constant() {r=1}
   ) {materialFunction=/Game/Effects/Fort_Effects/Materials/_Functions/MF_LowMobile_PathingSwitch}.Result) {name=Scanlines, nodeColor=<rgba 1 0 0.1641 1>}
  n22 = Multiply(A=NamedRerouteUsage()) {constB=0.2}
  n23 = NamedRerouteUsage()
  n24 = Multiply(
   A=Add(A=Multiply(A=@n19, B=@n21), B=Multiply(A=@n22, B=@n23))
   B=ScalarParameter() {defaultValue=1, group=Clawpad Host, parameterName=Global Emissive}
  )
  n25 = Fresnel(ExponentIn=ScalarParameter() {defaultValue=5, group=Clawpad Host, parameterName=Outline Fresnel}) {baseReflectFraction=0}
  n26 = Multiply(
   A=Step(Y=OneMinus(None=@n25), X=ScalarParameter() {defaultValue=0.8763, group=Clawpad Host, parameterName=Outline Step})
   B=ScalarParameter() {group=Clawpad Host, parameterName=Outline Opacity}
  )
  n27 = NamedRerouteUsage()
  n28 = NamedRerouteUsage()
  n29 = TextureSample() {
   samplerType=SAMPLERTYPE_Grayscale
   texture=/BRCosmetics/Accessories/FORT_Backpacks/Backpack_M_MED_ClawPad_Host/Textures/T_ClawPad_Host_Backpack_FX2
  }
  n30 = LinearInterpolate(A=NamedRerouteUsage(), B=@n29.R, Alpha=NamedRerouteUsage())
  n31 = ScalarParameter() {defaultValue=1, group=Clawpad Host, parameterName=Head Light Intensity}
  n32 = MaterialFunctionCall(
   Value_Morning=ScalarParameter() {defaultValue=3, group=Clawpad Host, parameterName=Darkening}
   Value_Day=ScalarParameter() {defaultValue=2, group=Clawpad Host, parameterName=Day Brightness}
   Value_Evening=ScalarParameter() {defaultValue=3, group=Clawpad Host, parameterName=Evening Brightness}
   Value_Night=ScalarParameter() {defaultValue=3, group=Clawpad Host, parameterName=Night Brightness}
  ) {materialFunction=&Scalar_Time_of_Day}
  n33 = LinearInterpolate(
   A=ScalarParameter() {defaultValue=0.4, group=Clawpad Host, parameterName=Sun Direction Multipler - Cody}
   B=ScalarParameter() {group=Clawpad Host, parameterName=Sun Direction Multipler - 66}
   Alpha=NamedRerouteUsage()
  )
  n34 = AtmosphericLightVector()
  n35 = VertexNormalWS()
  n36 = Multiply(A=Desaturation(None=@n1.BaseColor), B=LinearInterpolate(A=@n32.Result, B=Multiply(A=@n33, B=@n32.Result), Alpha=DotProduct(A=@n34, B=@n35)))
  n37 = NamedRerouteUsage()
  n38 = Multiply(
   A=Add(A=Multiply(A=@n30, B=@n31), B=Multiply(A=@n36, B=@n37))
   B=LinearInterpolate(
    A=VectorParameter() {group=Clawpad Host, parameterName=Holo Tint - Cody, defaultValue=<rgba 0.3927 0.8255 1 1>}.RGB
    B=VectorParameter() {group=Clawpad Host, parameterName=Holo Tint - 66, defaultValue=<rgba 0.3927 0.8255 1 1>}.RGB
    Alpha=NamedRerouteUsage()
   )
  )
  n39 = ScalarParameter() {defaultValue=1, group=Clawpad Host, parameterName=Enable Reactivity?}
  n40 = Multiply(
   A=NamedRerouteUsage()
   B=LinearInterpolate(B=ScalarParameter() {defaultValue=1, group=Clawpad Host, parameterName=66 Visibility}, Alpha=@n39)
  )
  n41 = Multiply(
   A=NamedRerouteUsage()
   B=LinearInterpolate(B=ScalarParameter() {defaultValue=1, group=Clawpad Host, parameterName=Cody Visibility}, Alpha=@n39) {constA=1}
  )
  n42 = NamedRerouteUsage()
  n43 = Multiply(A=NamedRerouteUsage(), B=Multiply(A=Add(A=@n40, B=@n41), B=Saturate(None=@n42)))
  n44 = ScalarParameter() {defaultValue=1, group=Clawpad Host, parameterName=Scanline Fresnel}
  n45 = Multiply(A=OneMinus(None=Fresnel(ExponentIn=@n44) {baseReflectFraction=0}), B=NamedRerouteUsage())
  n46 = NamedRerouteUsage()
  n47 = NamedRerouteUsage()
  n48 = Saturate(None=Subtract(A=Multiply(A=@n43, B=@n45), B=Multiply(A=@n46, B=@n47)))
  n49 = MaterialFunctionCall!(Opacity=@n48) {materialFunction=/Game/Global/Materials/MaterialGym/Materials/SortedOutline/MF_SortTranslucency}
  n50 = Add(
   A=Multiply(
    A=StaticSwitchParameter(True=@n49.Result, False=@n48) {group=Clawpad Host, parameterName=Use Sort Transluency?}
    B=ScalarParameter() {defaultValue=1, group=Clawpad Host, parameterName=Global Opacity}
   )
   B=NamedRerouteUsage()
  )
  n51 = ScalarParameter() {group=Clawpad Host, parameterName=WPO Intensity Min}
  n52 = ScalarParameter() {group=Clawpad Host, parameterName=WPO Intensity Max}
  n53 = NamedRerouteUsage()
  n54 = SetMaterialAttributes!(
   MaterialAttributes=@n1.MaterialAttributes
   EmissiveColor=LinearInterpolate(A=@n1.EmissiveColor, B=Add(A=@n24, B=@n26), Alpha=@n27)
   BaseColor=LinearInterpolate(A=@n1.BaseColor, B=Multiply(A=@n28, B=@n38), Alpha=@n27) {constB=0}
   Roughness=LinearInterpolate(A=@n1.Roughness, Alpha=@n27) {constB=2}
   Specular=LinearInterpolate(A=@n1.Specular, Alpha=@n27) {constB=0.5}
   OpacityMask=MaterialFunctionCall(Alpha Threshold=Saturate(None=@n50)) {materialFunction=&DitherTemporalAA}.Result
   Opacity=@n50
   WorldPositionOffset=Multiply(A=NamedRerouteUsage(), B=LinearInterpolate(A=@n51, B=@n52, Alpha=@n53)) {constB=0}
  ) {attributeSetTypes [
    #1
    #2
    #4
    #5
    #6
    #7
    #8
   ]}
  FunctionOutput(None=@n54) {outputName=SetMaterialAttributes}
  LocalPosition() {localOrigin=InstancePreSkinning}
  n55 = Divide(A=ObjectRadius(), B=ScalarParameter() {defaultValue=0.1, group=Clawpad Host, parameterName=NoiseSize})
  n56 = Multiply(A=Time(), B=VectorParameter() {group=Clawpad Host, parameterName=NoisePanSpeed, defaultValue=<rgba -0.08 0.07 -0.2 0>}.RGB) {constB=0}
  n57 = MaterialFunctionCall(
   TextureObject=TextureObject() {samplerType=SAMPLERTYPE_LinearColor, texture=&T_NoisePack_LinearColor}
   TextureSize=Multiply(A=VectorParameter() {group=Clawpad Host, parameterName=NoiseTileMult, defaultValue=<rgba 1 1 0.1 1>}.RGB, B=@n55)
   WorldPosition=Add(A=WorldPosition().XYZ, B=Multiply(A=@n55, B=@n56))
  ) {materialFunction=&WorldAlignedTexture}
  n58 = ActorPositionWS()
  n59 = MaterialFunctionCall(Float2=Round(None=Divide(A=@n58) {constB=2000})) {materialFunction=&BreakOutFloat2Components}
  n60 = Add(A=Time(), B=Add(A=@n59.R, B=@n59.R))
  n61 = Multiply(A=ScalarParameter() {defaultValue=1, group=Clawpad Host, parameterName=WPO - Flicker Speed}, B=@n60)
  n62 = Sine(None=@n61) {period=1.6}
  n63 = Sine(None=@n61) {period=1.42}
  n64 = Ceil(None=AppendVector(A=AppendVector(A=@n62, B=@n63), B=Sine(None=@n61) {period=1.51}))
  n65 = ScalarParameter() {defaultValue=1, group=Clawpad Host, parameterName=WPO - Noise Frequency}
  n66 = Add(A=Sine(None=@n61) {period=0.72}, B=Multiply(B=@n65) {constA=0.75}) {constB=0.75}
  n67 = Multiply(A=Add(A=Sine(None=@n61) {period=2.1}, B=Multiply(B=@n65) {constA=0.25}), B=@n66)
  n68 = Multiply(B=@n65) {constA=0.9, constB=0.9}
  n69 = Sine(None=@n61) {period=1.25}
  n70 = Multiply(A=@n66, B=Ceil(None=Add(A=@n68, B=@n69) {constB=0.9}))
  n71 = Sine(None=@n60) {period=0.075}
  n72 = Multiply(A=Saturate(None=Multiply(A=@n67, B=@n70)), B=LinearInterpolate(Alpha=ConstantBiasScale(None=@n71)) {constA=0.7})
  n73 = Add(
   A=ScalarParameter() {group=Clawpad Host, parameterName=WPO - NoiseAdd}
   B=Multiply(A=Multiply(A=@n64, B=@n72), B=ScalarParameter() {defaultValue=5, group=Clawpad Host, parameterName=WPO - NoiseMult})
  )
  n74 = Multiply(
   A=Multiply(A=Power(Base=@n57.XY Texture) {constExponent=4}, B=ScalarParameter() {defaultValue=2, group=Clawpad Host, parameterName=NoiseMaskMult}) {constB=20}
   B=@n73
  )
  n75 = MaterialFunctionCall(Float3=Desaturation(None=Saturate(None=@n74))) {materialFunction=&BreakOutFloat3Components}
  n76 = Multiply(A=Frac(None=Time()), B=ScalarParameter() {defaultValue=2, group=Clawpad Host, parameterName=WPO - Glitch Speed})
  n77 = Add(A=LocalPosition().Z, B=@n76)
  n78 = MaterialFunctionCall!(
   Width=ScalarParameter() {defaultValue=0.53, group=Clawpad Host, parameterName=WPO - Width}
   Sharpness=ScalarParameter() {defaultValue=0.6, group=Clawpad Host, parameterName=WPO - Sharpness}
   Offset=ScalarParameter() {group=Clawpad Host, parameterName=WPO - Offset}
   Direction Switch
   Compare=ScalarParameter() {group=Clawpad Host, parameterName=WPO - Compare}
   Bands=ScalarParameter() {defaultValue=0.06, group=Clawpad Host, parameterName=WPO - Bands}
   Input Coordinates=@n77
  ) {materialFunction=/Engine/Functions/Engine_MaterialFunctions03/Procedurals/GeneratedOffsetBands}
  n79 = LinearInterpolate(B=Multiply(A=@n73, B=@n78.Result), Alpha=NamedRerouteUsage()) {constA=1}
  n80 = Constant()
  n81 = MaterialFunctionCall(
   X=Multiply(
    A=@n75.R
    B=MaterialFunctionCall(R=@n79, G=@n80, B=@n80) {materialFunction=&AppendMany}.RGB
   )
   Y=@n75.G
   Z=@n75.G
  ) {materialFunction=&MakeFloat3}
  n82 = VertexColor()
  n83 = Subtract(
   A=WorldPosition() {worldPositionShaderOffset=WPT_ExcludeAllShaderOffsets}.XYZ
   B=MaterialFunctionCall() {materialFunction=&ObjectPivotPoint}.Object Pivot Location
  )
  n84 = ScalarParameter() {defaultValue=-6, parameterName=Cody Collapse}
  n85 = Add(A=@n81.Result, B=Multiply(A=@n82.B, B=Multiply(A=@n83, B=@n84)))
  n86 = Multiply(
   A=Subtract(
    A=WorldPosition() {worldPositionShaderOffset=WPT_ExcludeAllShaderOffsets}.XYZ
    B=MaterialFunctionCall() {materialFunction=&ObjectPivotPoint}.Object Pivot Location
   )
   B=Constant() {r=-5}
  ) {constB=-5}
  n87 = LinearInterpolate(
   A=Add(A=Multiply(A=@n86, B=@n82.R), B=@n81.Result)
   B=@n81.Result
   Alpha=ScalarParameter() {defaultValue=1, group=Clawpad Host, parameterName=66 Visibility}
  )
  n88 = ScalarParameter() {defaultValue=1, group=Clawpad Host, parameterName=Cody Visibility}
  NamedRerouteDeclaration(None=LinearInterpolate(
    A=@n81.Result
    B=LinearInterpolate(A=@n85, B=@n87, Alpha=@n88)
    Alpha=ScalarParameter() {defaultValue=1, group=Clawpad Host, parameterName=Enable Reactivity?}
   )) {name=WPO, nodeColor=<rgba 0 1 0.25 1>}
  NamedRerouteDeclaration(None=Step(Y=ScalarParameter() {defaultValue=0.9, group=Clawpad Host, parameterName=WPO - Glitch Step}, X=Sine(None=@n78.Result))) {name=Stripe, nodeColor=<rgba 0 1 0.7188 1>}
  n89 = TextureSample() {
   samplerType=SAMPLERTYPE_Masks
   texture=/BRCosmetics/Accessories/FORT_Backpacks/Backpack_M_MED_ClawPad_Host/Textures/T_ClawPad_Host_Backpack_FX
  }
  NamedRerouteDeclaration(None=@n89.B) {name=Base Mask, nodeColor=<rgba 0.875 0 1 1>}
  NamedRerouteDeclaration(None=Add(
    A=NamedRerouteDeclaration(None=@n89.R) {name=Body Mask, nodeColor=<rgba 1 0.2344 0 1>}
    B=NamedRerouteDeclaration(None=@n89.G) {name=66 Mask, nodeColor=<rgba 0 1 0.5547 1>}
   )) {name=Emissive Mask, nodeColor=<rgba 0 1 0.4609 1>}
  NamedRerouteDeclaration(None=@n89.A) {name=Dark Areas, nodeColor=<rgba 0.6172 1 0 1>}
  NamedRerouteDeclaration(None=MaterialFunctionCall(
    Width=ScalarParameter() {defaultValue=0.06, group=Clawpad Host, parameterName=Glitch Band - Width}
    Sharpness=ScalarParameter() {defaultValue=0.5, group=Clawpad Host, parameterName=Glitch Band - Sharpness}
    Offset=ScalarParameter() {defaultValue=-0.3905, group=Clawpad Host, parameterName=Glitch Band - Offset}
    Compare=@n76
    Bands=ScalarParameter() {defaultValue=0.15, group=Clawpad Host, parameterName=Glitch Band - Bands}
    Input Coordinates=@n77
   ) {materialFunction=&GeneratedOffsetBands}.Result) {name=Swipe, nodeColor=<rgba 1 0.7734 0 1>}
  n90 = ScalarParameter() {defaultValue=0.4, group=Clawpad Host, parameterName=Bottom Opacity Fade - Distance}
  NamedRerouteDeclaration(None=Divide(A=Add(A=@n7.B, B=@n90), B=ScalarParameter() {defaultValue=0.5, group=Clawpad Host, parameterName=Bottom Opacity Fade - Min})) {name=Fade, nodeColor=<rgba 0.6406 1 0 1>}
  NamedRerouteDeclaration(None=LinearInterpolate(
    B=ScalarParameter() {group=Clawpad Host, parameterName=Glitch Transition}
    Alpha=ScalarParameter() {defaultValue=1, group=Clawpad Host, parameterName=Enable Reactivity?}
   )) {name=Reactivity, nodeColor=<rgba 1 0.7734 0 1>}
  n91 = Fresnel(ExponentIn=ScalarParameter() {defaultValue=8, group=Clawpad Host, parameterName=Vertical Gradient - Fresnel}) {baseReflectFraction=0}
  n92 = WorldPosition()
  n93 = ObjectPositionWS()
  n94 = Divide(A=ComponentMask(None=Subtract(A=@n92.XYZ, B=@n93)) {b=true}, B=ObjectRadius())
  n95 = Saturate(None=Divide(A=ConstantBiasScale(None=@n94), B=ScalarParameter() {defaultValue=4, group=Clawpad Host, parameterName=Vertical Gradient - Sharpness}))
  n96 = OneMinus(None=@n95)
  n97 = ScalarParameter() {group=Clawpad Host, parameterName=Vertical Gradient - Flip Gradient}
  NamedRerouteDeclaration(None=LinearInterpolate(
    A=Step(Y=@n91) {constX=0.2}
    B=LinearInterpolate(A=@n95, B=@n96, Alpha=@n97)
    Alpha=ScalarParameter() {defaultValue=0.4, group=Clawpad Host, parameterName=Vertical Gradient - Influence}
   )) {name=Top, nodeColor=<rgba 0 0.2734 1 1>}
  n98 = Subtract(A=Multiply(A=@n12) {constB=-1}, B=Multiply(A=@n8) {constB=10})
  n99 = DotProduct(
   A=Constant3Vector() {constant=<rgba 0 0 -1 0>}
   B=Transform(None=VertexNormalWS()) {transformSourceType=TRANSFORMSOURCE_World, transformType=TRANSFORM_Local}
  )
  n100 = SmoothStep(Value=Add(A=@n7.B, B=ScalarParameter() {defaultValue=-0.22, group=Clawpad Host, parameterName=Head Mask - Distance})) {constMax=0, constMin=0.1}
  n101 = Subtract(A=@n99, B=@n100)
  n102 = Multiply(
   A=Multiply(A=Max(A=@n98, B=@n101)) {constB=-1}
   B=ScalarParameter() {defaultValue=2, group=Clawpad Host, parameterName=Shadow - Mask Intensity}
  )
  n103 = ScalarParameter() {defaultValue=0.7, group=Clawpad Host, parameterName=Shadow - Darkness Level}
  NamedRerouteDeclaration(None=Saturate(None=Max(A=@n102, B=@n103))) {name=Shadoow, nodeColor=<rgba 1 0 0.1406 1>}
  n104 = Multiply(A=@n99) {constB=-1}
  NamedRerouteDeclaration(None=Saturate(None=Subtract(A=@n104, B=@n100))) {name=Top Light, nodeColor=<rgba 1 0 0.2109 1>}
  NamedRerouteDeclaration(None=LinearInterpolate(
    B=ScalarParameter() {defaultValue=1, group=Clawpad Host, parameterName=66 Visibility}
    Alpha=ScalarParameter() {defaultValue=1, group=Clawpad Host, parameterName=Enable Reactivity?}
   )) {name=66 Reactivity, nodeColor=<rgba 1 0.7266 0 1>}
  Saturate(None=NamedRerouteUsage())
  Saturate(None=NamedRerouteUsage())
 ]}
